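; Holiday prop Blueprint for the Lumpkin King A decoration in the Helios environment, with no gameplay logic of its own.
; It overrides parent functions to supply the body component's static mesh for collision at a zero vector and list both the StaticMeshComponent and body as mesh components.
; IgnoreWorldGridScaleOverride returns false, the construction script only calls its parent, and BeginPlay, overlap and Tick events are empty.
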(event EventBeginPlay)

(event Collision|EventActorBeginOverlap
  (params (OtherActor :type Actor%20Object%20Reference)))

(event EventTick
  (params (DeltaSeconds :type Float%20%28single-precision%29)))

(fn GetCollisionStaticMesh
  (return "0, 0, 0" (Variables|/Script/Engine.StaticMeshComponent|GetStaticMesh (Variables|Default|GetSM_Lumpkin_King_A_Body))))

(fn GetMeshComponents
  (return (Utilities|Array|MakeArray (Variables|Debug|GetStaticMeshComponent) (Variables|Default|GetSM_Lumpkin_King_A_Body))))

(fn IgnoreWorldGridScaleOverride
  (return false))

(fn ConstructionScript
  (|CallParentFunction))
