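# MF_LKEmissive is a material function for the Holiday Lumpkin King prop, named for emissive shading.
# Its expression graph is empty, with no inputs, outputs or logic shown.
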MaterialFunction /Game/Environments/Helios/Props/Holiday/Holiday_LumpkinKing_A/Materials/Functions/MF_LKEmissive
graph {expressions {}}
