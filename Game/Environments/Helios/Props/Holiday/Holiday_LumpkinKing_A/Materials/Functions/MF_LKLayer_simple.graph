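# Material function for the Lumpkin King holiday prop that builds a simple material attributes layer from albedo, normal and AOSR texture inputs.
# All three textures sample on one UV input with shared world-group wrap samplers, linked through named reroutes.
# Albedo drives base color and normal feeds through a reroute.
# Channel mask parameters pick AO, specular and roughness from the packed AOSR map, defaulting to specular in green and roughness in blue.
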MaterialFunction /Game/Environments/Helios/Props/Holiday/Holiday_LumpkinKing_A/Materials/Functions/MF_LKLayer_simple
graph {nodes [
  n1 = NamedRerouteDeclaration(None=FunctionInput() {inputName=UVs, inputType=FunctionInput_Vector2}) {name=UVs, nodeColor=<rgba 0.9922 0 1 1>}
  n2 = TextureObject()
  n3 = NamedRerouteDeclaration(None=FunctionInput(Preview=@n2) {inputName=Tobj_Albedo, inputType=FunctionInput_Texture2D, sortPriority=1}) {name=albedoOBJ, nodeColor=<rgba 1 0.75 0 1>}
  n4 = FunctionInput(Preview=@n2) {inputName=Tobj_AOSR, inputType=FunctionInput_Texture2D, sortPriority=3}
  n5 = NamedRerouteDeclaration(None=TextureSample(UVs=@n1, Tex=NamedRerouteDeclaration(None=@n4) {name=AOSROBJ, nodeColor=<rgba 0.4297 0 1 1>}) {samplerSource=SSM_Wrap_WorldGroupSettings}.RGB) {name=AOSRUV, nodeColor=<rgba 0 0.5312 1 1>}
  n6 = ChannelMaskParameter(None=@n5) {group=B_GlobalMaskSelector, maskChannel=Green, parameterName=Specular Channel, defaultValue=<rgba 0 1 0 0>}
  n7 = ChannelMaskParameter(None=@n5) {group=B_GlobalMaskSelector, maskChannel=Blue, parameterName=Roughness Channel, defaultValue=<rgba 0 0 1 0>}
  n8 = SetMaterialAttributes!(
   MaterialAttributes
   BaseColor=NamedRerouteDeclaration(None=TextureSample(UVs=@n1, Tex=@n3) {samplerSource=SSM_Wrap_WorldGroupSettings}.RGB) {name=albedoUV, nodeColor=<rgba 1 0 0.2109 1>}
   Specular=@n6.R
   Roughness=@n7.R
   AmbientOcclusion=ChannelMaskParameter(None=@n5) {group=B_GlobalMaskSelector, parameterName=AO Channel}.R
   Normal=NamedRerouteUsage()
  ) {attributeSetTypes [
    #1
    #2
    #3
    #4
    #5
   ]}
  FunctionOutput(None=@n8)
  n9 = FunctionInput(Preview=TextureObject() {samplerType=SAMPLERTYPE_Normal, texture=/Engine/EngineMaterials/FlatNormal}) {inputName=Tobj_Normal, inputType=FunctionInput_Texture2D, sortPriority=2}
  NamedRerouteDeclaration(None=TextureSample(UVs=@n1, Tex=NamedRerouteDeclaration(None=@n9) {name=normalOBJ, nodeColor=<rgba 0 1 0.1094 1>}).RGB) {name=normalUV, nodeColor=<rgba 0.8516 1 0 1>}
 ]}
